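# Parent body material for the Quake skin of the male medium BraveBuild player character, wrapping the shared 2023 CharacterParent PreFX and PostFX functions.
# Between the two functions, it multiplies the Opacity attribute by the red channel of a SkinOpacity mask, which defaults to white.
# A quality switch raises that result to a tunable power on Epic only, then writes it back into the material attributes for PostFX.
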Material /BRCosmetics/Characters/Player/Male/Medium/Bodies/M_MED_BraveBuild/Skins/Quake/Materials/M_MED_BraveBuild_Quake_Body_Parent
graph {
 properties {
  bUseMaterialAttributes=true
  bUsedWithMeshDeformer=true
  bUsedWithMorphTargets=true
  bUsedWithNanite=true
  bUsedWithSkeletalMesh=true
  pixelDepthOffsetMode=PDOM_Legacy
  shadingModel=MSM_FromMaterialExpression
  usedShadingModels=Default Lit | Subsurface | Clear Coat | Subsurface Profile | Cloth
 }
 nodes [
  n1 = MaterialFunctionCall() {materialFunction=&MF_CharacterParent_2023_PreFX}
  n2 = GetMaterialAttributes(None=@n1.Result) {attributeGetTypes [#1, #2]}
  n3 = TextureSampleParameter2D() {group=BraveBuildQuake, parameterName=SkinOpacity, samplerType=SAMPLERTYPE_Masks, texture=/Game/Global/Textures/Default/Blanks/T_White}
  n4 = Multiply(A=@n2.Opacity, B=@n3.R)
  n5 = Power(Base=@n4, Exp=ScalarParameter() {defaultValue=1, group=BraveBuildQuake, parameterName=SkinOpacityPowOnEpic})
  MaterialFunctionCall(
   Material=SetMaterialAttributes!(MaterialAttributes=@n2.MaterialAttributes, Opacity=QualitySwitch(Default=@n4, Epic=@n5)) {attributeSetTypes [#2]}
   M=@n1.M
   Pre Skinned Local Position=@n1.M
   Distance Blend=@n1.M
   Fresnel=@n1.M
  ) {materialFunction=&MF_CharacterParent_2023_PostFX}
 ]
}
