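# Parent head material for the Quake skin of the male medium BraveBuild character, with multiple shading models including subsurface, eye and cloth.
# CharacterSkinFX output has its opacity multiplied by a SkinOpacity mask, then passes through character PostFX functions.
# On Epic quality, a rounded, inverted emissive red channel selects between that opacity and a power-adjusted version.
# A separate mask power branch feeds nothing.
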Material /BRCosmetics/Characters/Player/Male/Medium/Bodies/M_MED_BraveBuild/Skins/Quake/Materials/M_MED_BraveBuild_Quake_Head_Parent
graph {
 properties {
  bUseMaterialAttributes=true
  bUsedWithMeshDeformer=true
  bUsedWithNanite=true
  bUsedWithSkeletalMesh=true
  pixelDepthOffsetMode=PDOM_Legacy
  shadingModel=MSM_FromMaterialExpression
  usedShadingModels=Default Lit | Subsurface | Clear Coat | Subsurface Profile | Cloth | Eye
 }
 nodes [
  n1 = MaterialFunctionCall() {materialFunction=&BaseFN_CharacterFunction_PreFX}
  n2 = GetMaterialAttributes(None=MaterialFunctionCall(
    PreskinnedPosition=@n1.Pre skinned Local Position
    Gmap=@n1.Pre skinned Local Position
    MaterialAttributes=@n1.Pre skinned Local Position
    AO=@n1.Pre skinned Local Position
   ) {materialFunction=&CharacterSkinFX}.Result) {attributeGetTypes [#1, #2]}
  n3 = TextureSampleParameter2D() {group=BraveBuildQuake, parameterName=SkinOpacity, samplerType=SAMPLERTYPE_Masks, texture=/Game/Global/Textures/Default/Blanks/T_White}
  n4 = Multiply(A=@n2.Opacity, B=@n3.R)
  n5 = Power(Base=@n4, Exp=ScalarParameter() {defaultValue=1, group=BraveBuildQuake, parameterName=SkinOpacityPowOnEpic})
  n6 = Round(None=OneMinus(None=ComponentMask(None=@n2.Emissive Color) {r=true}))
  n7 = SetMaterialAttributes!(
   MaterialAttributes=@n2.MaterialAttributes
   Opacity=QualitySwitch(Default=@n4, Epic=LinearInterpolate(A=@n5, B=@n4, Alpha=@n6))
   Base Color
   Emissive Color
  ) {attributeSetTypes [#2, #3, #1]}
  n8 = MaterialFunctionCall() {materialFunction=&MF_CharacterParent_2023_PreFX}
  MaterialFunctionCall(Material=@n7, M=@n8.M, Pre Skinned Local Position=@n8.M, Distance Blend=@n8.M, Fresnel=@n8.M) {materialFunction=&MF_CharacterParent_2023_PostFX}
  n9 = MaterialFunctionCall(AO=@n1.AO, Distance Blend=@n1.AO, Fresnel=@n1.AO, Gmap=@n1.AO, Pre Skinned Local Position=@n1.AO, Material=@n7) {materialFunction=&BaseFN_CharacterFunction_PostFX}
  MaterialFunctionCall(In=@n9.Result) {materialFunction=&CINE_MF_CharPostFX}
  Power(Base=@n3.R, Exp=ScalarParameter() {defaultValue=1, group=BraveBuildQuake, parameterName=SkinOpacityMaskPowOnEpic})
 ]
}
